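# Niagara effect for the FrameDash player emote, producing foot dust and circular bursts as mesh particles on the skeletal mesh.
# User floats BurstIndex and BurstCubes each act as trigger and bone/socket selector: a change in their integer value fires a spawn there.
# Dust_Feet layers four spawn groups, tagged for separate renderers, thrown along -Y and shrinking.
# Burst_Circle pushes particles outward from a cylinder surface; both emitters tint by time of day.
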NiagaraSystem /BRCosmetics/Animation/Game/MainPlayer/Emotes/FrameDash/FX/NS_Emote_FrameDash
dependencies {
 usedDataInterfaces [&NiagaraDataInterfaceCurve, &NiagaraDataInterfaceSkeletalMesh]
 usedDynamicInputs []
 usedModules [
  &self:SystemSpawnScript.SystemScriptSource.SystemScriptGraph.NiagaraNodeAssignment_1.SetVariables_#1
  &SystemState
 ]
}
emitters {
 Burst_Circle {
  emitterUpdateScript [
   &EmitterState(Life Cycle Mode=::NewEnumerator1, Loop Behavior=::NewEnumerator1, Loop Delay=6.4, Loop Duration Mode=::NewEnumerator1, UseLoopDelay=false)
   /Game/Effects/Niagara/Modules/Logic/UpdateValueStateChange(InputIntParam:NiagaraInt32=&MakeIntFromFloat(Float=$User.BurstCubes, Conversion Mode=::NewEnumerator1))
   /Niagara/Modules/Emitter/SpawnByTrigger(Spawn Count=45, Trigger:NiagaraBool=$Emitter.UpdateValueStateChange.Changed, TriggerIntervalSwitch=false)
  ]
  particleSpawnScript [
   &InitializeParticle(
    AdjustAlpha=false
    AdjustHue=false
    AdjustSaturation=false
    AdjustValue=false
    Color=&FNTimeOfDayColor(
     Night Color=<rgba 0.9757 1.1754 1.7 1>
     Morning Color=<rgba 1 0.5417 0.551 1>
     Day Color=<rgba 1 1 1 1>
     Evening Color=<rgba 1 0.5837 0.7216 1>
    )
    Color Mode=::NewEnumerator3
    Lifetime=&RandomRangeFloat(Minimum=0.17, Maximum=0.35)
    Mesh Scale Mode=::NewEnumerator3
    Mesh Uniform Scale=&RandomRangeFloat(Minimum=0.1, Maximum=0.18)
    UsePositionOffset=false
    Value Range:Vector2f=<0.8 1.2>
    Write Material Random=true
   )
   &SkeletalMeshLocation(
    Bone / Socket Index=&MakeIntFromFloat(Float=$User.BurstCubes, Conversion Mode=::NewEnumerator1)
    Bone Sampling Mode=::NewEnumerator4
    MaskSpawnGroup=false
    Orientation Sampling=::NewEnumerator1
    Skeletal Mesh=$System.Skeletal Mesh Reader
   )
   &ShapeLocation(
    Cylinder Height=1
    Cylinder Radius=&RandomRangeFloat(Minimum=7, Maximum=25)
    MaskSpawnGroup=false
    Offset=&MergeHelper_VectorSwitch(
     0=<-10 20 -20>
     1=<10 20 -20>
     2=?
     3=?
     Index=&MakeIntFromFloat(Float=$User.BurstCubes, Conversion Mode=::NewEnumerator1)
     Num=::NewEnumerator0
    )
    Offset Coordinate Space=::Local
    Offset Mode=::NewEnumerator0
    Override Seed=false
    Shape Primitive=::NewEnumerator1
    Surface Only=true
    Use Endcaps In Surface Only Mode=false
    UseOwnerScale=true
   )
   &AddVelocity(
    Rotation Coordinate Space=::Simulation
    Velocity=$Particles.ShapeLocation.ShapeNormal
    Velocity Speed Scale=&RandomRangeFloat(Minimum=75, Maximum=150)
   )
  ]
  particleUpdateScript [
   &ParticleState()
   &Drag(Drag=&RandomRangeFloat(Minimum=1.5, Maximum=3), Use Linear Drag=true, Use Rotational Drag=true)
   &SolveForcesAndVelocity(Clamp Velocity=false, Limit Acceleration=false)
   &KillParticles(Kill Particles=&SetBoolByFloatComparison(A=$System.Age, B=0.3, Comparison Type=::NewEnumerator6))
  ]
  rendererClasses [&NiagaraMeshRendererProperties]
  renderers [{rendererClass=&NiagaraMeshRendererProperties, rendererIndex=0}]
 }
 Dust_Feet {
  emitterUpdateScript [
   &EmitterState(Life Cycle Mode=::NewEnumerator1, Loop Behavior=::NewEnumerator1, Loop Duration Mode=::NewEnumerator1, UseLoopDelay=false)
   &UpdateValueStateChange(InputIntParam:NiagaraInt32=&MakeIntFromFloat(Float=$User.BurstIndex, Conversion Mode=::NewEnumerator1))
   &SpawnByTrigger(
    Spawn Count=12
    Trigger:NiagaraBool=&Boolean_AndOperation(Bool1=$Emitter.UpdateValueStateChange.Changed, Bool2=&SetBoolByFloatComparison(A=$System.Age, B=0.1))
    TriggerIntervalSwitch=false
   )
   &SpawnByTrigger(
    Spawn Count=5
    Spawn Group=1
    Trigger:NiagaraBool=&Boolean_AndOperation(Bool1=$Emitter.UpdateValueStateChange.Changed, Bool2=&SetBoolByFloatComparison(A=$System.Age, B=0.1))
    TriggerIntervalSwitch=false
   ) {module=SpawnByTrigger001}
   &SpawnByTrigger(
    Spawn Count=5
    Spawn Group=2
    Trigger:NiagaraBool=&Boolean_AndOperation(Bool1=$Emitter.UpdateValueStateChange.Changed, Bool2=&SetBoolByFloatComparison(A=$System.Age, B=0.1))
    TriggerIntervalSwitch=false
   ) {module=SpawnByTrigger002}
   &SpawnByTrigger(
    Spawn Count=20
    Spawn Group=3
    Trigger:NiagaraBool=&Boolean_AndOperation(Bool1=$Emitter.UpdateValueStateChange.Changed, Bool2=&SetBoolByFloatComparison(A=$System.Age, B=0.1))
    TriggerIntervalSwitch=false
   ) {module=SpawnByTrigger003}
  ]
  particleSpawnScript [
   /SetVariables(Particles.VisibilityTag=&SpawnGroup()) {isSetParameters=true}
   &PartitionParticles(Particles Per Partition=10, Partition Mode:ENiagara_PartitionMode_v2=::NewEnumerator1)
   &InitializeParticle(
    AdjustAlpha=false
    AdjustHue=false
    AdjustSaturation=false
    AdjustValue=false
    Color=&FNTimeOfDayColor(
     Night Color=<rgba 0.9757 1.1754 1.7 1>
     Morning Color=<rgba 1 0.5417 0.551 1>
     Day Color=<rgba 1 1 1 1>
     Evening Color=<rgba 1 0.5837 0.7216 1>
    )
    Color Mode=::NewEnumerator3
    Lifetime=&MergeHelper_FloatSwitch(
     0=&RandomRangeFloat(Minimum=0.3, Maximum=0.5)
     1=&RandomRangeFloat(Minimum=0.3, Maximum=0.5)
     2=&RandomRangeFloat(Minimum=0.3, Maximum=0.5)
     3=&RandomRangeFloat(Minimum=0.25, Maximum=0.4)
     Index=&SpawnGroup()
     Num=::NewEnumerator2
    )
    Mesh Scale Mode=::NewEnumerator3
    Mesh Uniform Scale=&MergeHelper_FloatSwitch(
     0=&RandomRangeFloat(Minimum=0.22, Maximum=0.32)
     1=&RandomRangeFloat(Minimum=0.17, Maximum=0.24)
     2=&RandomRangeFloat(Minimum=0.17, Maximum=0.27)
     3=&RandomRangeFloat(Minimum=0.18, Maximum=0.22)
     Index=&SpawnGroup()
     Num=::NewEnumerator2
    )
    Sprite Size Mode=::NewEnumerator3
    Uniform Sprite Size=80
    UsePositionOffset=false
    Value Range:Vector2f=<0.8 1.2>
    Write Material Random=true
   )
   &SkeletalMeshLocation(
    Bone / Socket Index=&MakeIntFromFloat(Float=$User.BurstIndex, Conversion Mode=::NewEnumerator1)
    Bone Sampling Mode=::NewEnumerator4
    MaskSpawnGroup=false
    Skeletal Mesh=$System.Skeletal Mesh Reader
   )
   &ShapeLocation(
    MaskSpawnGroup=false
    Offset=&MergeHelper_VectorSwitch(0=<0 -40 12>, 1=<0 -25 8>, 2=<0 -37 24>, 3=<0 0 0>, 4=<0 0 0>, Index=&SpawnGroup(), Num=::NewEnumerator3)
    Offset Coordinate Space=::Local
    Offset Mode=::NewEnumerator0
    Sphere Distribution=::NewEnumerator3
    Sphere Radius=&MergeHelper_FloatSwitch(
     0=&RandomRangeFloat(Minimum=8.5, Maximum=9)
     1=&RandomRangeFloat(Minimum=3.5, Maximum=4)
     2=&RandomRangeFloat(Minimum=3.5, Maximum=4)
     3=0
     Index=&SpawnGroup()
     Num=::NewEnumerator2
    )
    UseOwnerScale=true
   )
   &ShapeLocation(
    Cone Angle=9
    Cone Length:NiagaraFloat=&MaskFloat(Mask Source=0, Mask Parameter=?, Mask=3, Masked Float=50, Passthrough Float=1)
    MaskSpawnGroup=false
    Non Uniform Scale=<2.2 1 1>
    Offset=&MaskVector(Mask Source=0, Mask Parameter=?, Mask=3, Masked Vector=<0 10 0>, Passthrough Vector=<1 1 1>)
    Offset Coordinate Space=::Local
    Offset Mode=::NewEnumerator0
    Override Seed=false
    Rotation Axis=<0 -1 0.03>
    Rotation Mode=::NewEnumerator1
    Shape Primitive=::NewEnumerator5
    UseOwnerScale=true
   ) {module=ShapeLocation001}
   &AddVelocity(
    Constrain To Radius=20
    Module.Use Falloff Distance=false
    Origin Offset=<0 -37 0>
    Override Seed=false
    Velocity Mode=::NewEnumerator1
    Velocity Speed=50
   ) {module=AddVelocity001}
   &AddVelocity(Velocity=<0 -380 0>)
  ]
  particleUpdateScript [
   &ParticleState()
   &ScaleSpriteSize(Uniform Curve Sprite Scale=DataInterface() {propertyValues {Curve {
       defaultValue=0
       keys [{interpMode=RCIM_Cubic, tangentMode=RCTM_User, time=0, value=0.3}, {interpMode=RCIM_Cubic, tangentMode=RCTM_User, time=1, value=1}]
      }}})
   &ScaleMeshSize(Scale Factor=&VectorFromFloat(Value=&FloatFromCurve(
      FloatCurve=DataInterface() {propertyValues {Curve {
         defaultValue=1
         keys [{interpMode=RCIM_Cubic, tangentMode=RCTM_User, time=0, value=1}, {interpMode=RCIM_Cubic, tangentMode=RCTM_User, time=1, value=0}]
        }}}
      Scale Curve=&RandomRangeFloat(Minimum=1, Maximum=1.2)
     )))
   &Drag(Drag=4, Use Linear Drag=true, Use Rotational Drag=false)
   &SolveForcesAndVelocity(Clamp Velocity=false, Limit Acceleration=false)
   &ScaleColor(
    Scale Alpha=&MergeHelper_FloatSwitch(
     0=1
     1=&FloatFromCurve(FloatCurve=DataInterface() {propertyValues {Curve {
         defaultValue=1
         keys [{interpMode=RCIM_Cubic, tangentMode=RCTM_User, time=0, value=1}, {interpMode=RCIM_Cubic, tangentMode=RCTM_User, time=1, value=0}]
        }}})
     2=?
     3=?
     Index=&SpawnGroup()
     Num=::NewEnumerator0
    )
    ScaleA=true
    ScaleRGB=true
   )
  ]
  rendererClasses [&NiagaraMeshRendererProperties]
  renderers [
   {rendererClass=&NiagaraMeshRendererProperties, rendererIndex=0}
   {rendererClass=&NiagaraMeshRendererProperties, rendererIndex=1}
   {rendererClass=&NiagaraMeshRendererProperties, rendererIndex=2}
   {rendererClass=&NiagaraMeshRendererProperties, rendererIndex=3}
  ]
 }
}
summary {
 systemName=NS_Emote_FrameDash
 emitters [
  {emitterName=Dust_Feet, rendererClasses [&NiagaraMeshRendererProperties]}
  {emitterName=Burst_Circle, rendererClasses [&NiagaraMeshRendererProperties]}
 ]
 userVariables [
  User.bIsFrontEnd {type=NiagaraBool}
  User.bIsFrontEndPreview {type=NiagaraBool}
  User.BurstCubes {type=NiagaraFloat, defaultValue=0, description=""}
  User.BurstIndex {type=NiagaraFloat, defaultValue=0, description=""}
  User.FrontEnd_color {type=LinearColor, description="", defaultValue {struct=LinearColor, value=<rgba 1 0.6743 0.6445 1>}}
 ]
}
